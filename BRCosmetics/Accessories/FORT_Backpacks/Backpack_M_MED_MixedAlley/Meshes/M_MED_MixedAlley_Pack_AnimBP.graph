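# Anim Blueprint for the MixedAlley backpack, a cosmetic with a floating ball that advances through staged animations on eliminations.
# With variant reactivity on, a kill-counter state machine picks idle and reaction sequences, and front-end delays cycle the stages.
# Anim Dynamics plus sine-driven bone offset and squash-stretch float the ball.
# Anim notifies toggle pom-pom and egg visibility through scalar parameters on dynamic base and overlay materials.
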AnimBlueprint /BRCosmetics/Accessories/FORT_Backpacks/Backpack_M_MED_MixedAlley/Meshes/M_MED_MixedAlley_Pack_AnimBP
target_skeleton=&/M_MED_MixedAlley_Pack_Skeleton
graphs [
 AnimGraph {nodes [
   Input Pose()
   n1 = Sequence Player 'MixedAlley_IdleAnim_5_PACK'() {bLoopAnimation=false, sequence=/BRCosmetics/Accessories/FORT_Backpacks/Backpack_M_MED_MixedAlley/Anim/MixedAlley_IdleAnim_5_PACK}
   AnimGraphNode_StateMachine_1 = OnKillStateMachine() {is=title AnimGraphNode_StateMachine}
   n2 = Get Cosmetic Variant Reactivity() -> (Cosmetic Variant Reactivity:bool)
   n3 = Anim Dynamics - Bone: BP_floatingBall_base(
    ComponentPose=Local To Component(LocalPose=Blend Poses by bool(BlendPose_0=@AnimGraphNode_StateMachine_1, BlendPose_1=@n1, bActiveValue=@n2) {blendType=Linear})
    Alpha=0.7
    ExternalForce="0,0,0"
   ) {
    angularDampingOverride=1
    angularSpringConstant=500
    bLinearSpring=true
    bOverrideLinearDamping=true
    bUseGravityOverride=true
    gravityScale=0.35
    linearDampingOverride=1
    linearSpringConstant=50
    simulationSpace=RootRelative
    boundBone {boneName=BP_floatingBall_base}
    componentAppliedLinearAccClamp=<2000 2000 2000>
    componentLinearAccScale=<0.3 0.3 0.1>
    componentLinearVelScale=<0.5 0.5 0.1>
    physicsBodyDefinitions [{
      boundBone {boneName=BP_floatingBall_base}
      boxExtents=<20 20 20>
      constraintSetup {angularLimitsMax=<0 0 0>, angularLimitsMin=<0 0 0>, angularTarget=<1 0 0>, linearAxesMax=<8 8 10>, linearAxesMin=<0 -8 -10>}
      localJointOffset=<0 0 0>
     }]
    relativeSpaceBone {boneName=BP_HeadTop}
    simSpaceSettings {maxAngularAcceleration=100, maxAngularVelocity=50, simSpaceAngularAlpha=1}
   }
   n4 = "Transform (Modify) Bone - Bone: BP_floatingBall"(ComponentPose=@n3, Alpha=1, Translation=Get FloatingBall_Movement() -> (FloatingBall_Movement:Vector), Rotation="0,0,0", Scale="1,1,1") {
    rotationMode=BMM_Additive
    rotationSpace=BCS_BoneSpace
    scaleMode=BMM_Replace
    translationMode=BMM_Additive
    translationSpace=BCS_BoneSpace
    alphaBoolBlend {blendInTime=0.2, blendOutTime=0.2}
    alphaScaleBiasClamp {interpSpeedDecreasing=5, interpSpeedIncreasing=5}
    boneToModify {boneName=BP_floatingBall}
   }
   n5 = Get StretchSquashFloatingBall() -> (StretchSquashFloatingBall:Vector)
   n6 = Local To Component(LocalPose=Component To Local(ComponentPose="Transform (Modify) Bone - Bone: BP_floatingBall"(ComponentPose=@n4, Alpha=1, Translation="0,0,0", Rotation="0,0,0", Scale=@n5) {rotationMode=BMM_Replace, scaleMode=BMM_Replace, boneToModify {boneName=BP_floatingBall}}))
   Output Pose(Result=Component To Local(ComponentPose="Transform (Modify) Bone - Bone: root"(ComponentPose=@n6, Translation="0,0,8", Rotation="0,0,0") {translationMode=BMM_Additive, translationSpace=BCS_BoneSpace, boneToModify {boneName=root}, translation {z=8}})) {layerGroup=None}
  ]}
 OnKillStateMachine {
  entry=Idle
  path=AnimGraph.AnimGraphNode_StateMachine_1.OnKillStateMachine
  ids=id
  nodes [
   AnimStateNode_0 = Idle() {
    stateEntered {guid=#1}
    stateFullyBlended {guid=#2}
    stateLeft {guid=#3}
   }
   AnimStateNode_1 = OnKill() {
    is=title AnimStateNode
    stateEntered {guid=#4}
    stateFullyBlended {guid=#5}
    stateLeft {guid=#6}
   }
   AnimStateTransitionNode_0 = AnimStateTransitionNode() {bAllowInertializationForSelfTransitions=true, crossfadeDuration=0}
   AnimStateTransitionNode_1 = AnimStateTransitionNode() {bAutomaticRuleBasedOnSequencePlayerInState=true}
   AnimStateTransitionNode_3 = AnimStateTransitionNode() {bAllowInertializationForSelfTransitions=true}
  ]
  links [
   {
    dst=OnKill
    rule=AnimGraph.AnimGraphNode_StateMachine_1.OnKillStateMachine.AnimStateTransitionNode_0.Transition
    src=Idle
    via=AnimStateTransitionNode_0
   }
   {
    dst=OnKill
    rule=AnimGraph.AnimGraphNode_StateMachine_1.OnKillStateMachine.AnimStateTransitionNode_3.Transition
    src=Idle
    via=AnimStateTransitionNode_3
   }
   {
    dst=Idle
    rule=AnimGraph.AnimGraphNode_StateMachine_1.OnKillStateMachine.AnimStateTransitionNode_1.Transition
    src=OnKill
    via=AnimStateTransitionNode_1
   }
  ]
 }
 Idle {
  path=AnimGraph.AnimGraphNode_StateMachine_1.OnKillStateMachine.AnimStateNode_0.Idle
  nodes [
   n1 = Blend Poses by int!(
    BlendPose_0:Pose Link Structure=Sequence Player 'MixedAlley_IdleAnim_1_PACK'() {sequence=/BRCosmetics/Accessories/FORT_Backpacks/Backpack_M_MED_MixedAlley/Anim/MixedAlley_IdleAnim_1_PACK}
    BlendPose_1:Pose Link Structure=Sequence Player 'MixedAlley_IdleAnim_2_PACK'() {sequence=/BRCosmetics/Accessories/FORT_Backpacks/Backpack_M_MED_MixedAlley/Anim/MixedAlley_IdleAnim_2_PACK}
    BlendPose_2:Pose Link Structure=Sequence Player 'MixedAlley_IdleAnim_3_PACK'() {sequence=/BRCosmetics/Accessories/FORT_Backpacks/Backpack_M_MED_MixedAlley/Anim/MixedAlley_IdleAnim_3_PACK}
    BlendPose_3:Pose Link Structure=Sequence Player 'MixedAlley_IdleAnim_4_PACK'() {sequence=/BRCosmetics/Accessories/FORT_Backpacks/Backpack_M_MED_MixedAlley/Anim/MixedAlley_IdleAnim_4_PACK}
    BlendPose_4:Pose Link Structure=Sequence Player 'MixedAlley_IdleAnim_5_PACK'() {sequence=&MixedAlley_IdleAnim_5_PACK}
    BlendTime_0:float=1
    BlendTime_1:float=1
    BlendTime_2:float=1
    BlendTime_3:float=1
    BlendTime_4:float=1
    ActiveChildIndex:int=Get Kill Counter() -> (Kill Counter:int=0)
   ) -> (Pose:Pose Link Structure) {
    is=title AnimGraphNode_BlendListByInt
    blendType=Linear
    blendPose [
     {linkId=-1, sourceLinkId=-1}
     {linkId=-1, sourceLinkId=-1}
     {linkId=-1, sourceLinkId=-1}
     {linkId=-1, sourceLinkId=-1}
     {linkId=-1, sourceLinkId=-1}
    ]
    blendTime [0.1,0.1,0.1,0.1,0.1]
   }
   Output Animation Pose(Result=@n1)
  ]
 }
 OnKill {
  path=AnimGraph.AnimGraphNode_StateMachine_1.OnKillStateMachine.AnimStateNode_1.OnKill
  nodes [
   n1 = Sequence Player 'MixedAlley_Stage1to2_Reactive_PACK'() {bLoopAnimation=false, sequence=/BRCosmetics/Accessories/FORT_Backpacks/Backpack_M_MED_MixedAlley/Anim/MixedAlley_Stage1to2_Reactive_PACK}
   n2 = Sequence Player 'MixedAlley_Stage2to3_Reactive_PACK'() {bLoopAnimation=false, sequence=/BRCosmetics/Accessories/FORT_Backpacks/Backpack_M_MED_MixedAlley/Anim/MixedAlley_Stage2to3_Reactive_PACK}
   n3 = Sequence Player 'MixedAlley_Stage3to4_Reactive_PACK'() {bLoopAnimation=false, sequence=/BRCosmetics/Accessories/FORT_Backpacks/Backpack_M_MED_MixedAlley/Anim/MixedAlley_Stage3to4_Reactive_PACK}
   n4 = Sequence Player 'MixedAlley_Stage4to5_Reactive_PACK'() {bLoopAnimation=false, sequence=/BRCosmetics/Accessories/FORT_Backpacks/Backpack_M_MED_MixedAlley/Anim/MixedAlley_Stage4to5_Reactive_PACK}
   n5 = Blend Poses by int!(
    BlendPose_0=Sequence Player 'MixedAlley_IdleAnim_1_PACK'() {sequence=&MixedAlley_IdleAnim_1_PACK}
    BlendPose_1=@n1
    BlendPose_2=@n2
    BlendPose_3=@n3
    BlendPose_4=@n4
    BlendPose_5:Pose Link Structure=Sequence Player 'MixedAlley_Reactive5_PACK'() {bLoopAnimation=false, sequence=/BRCosmetics/Accessories/FORT_Backpacks/Backpack_M_MED_MixedAlley/Anim/MixedAlley_Reactive5_PACK}
    BlendTime_0=0
    BlendTime_1=0
    BlendTime_2=0
    BlendTime_3=0
    BlendTime_4=0
    BlendTime_5:float=0
    ActiveChildIndex=Get Kill Counter() -> (Kill Counter=0)
   ) -> (Pose) {
    blendPose [
     {linkId=-1, sourceLinkId=-1}
     {linkId=-1, sourceLinkId=-1}
     {linkId=-1, sourceLinkId=-1}
     {linkId=-1, sourceLinkId=-1}
     {linkId=-1, sourceLinkId=-1}
     {linkId=-1, sourceLinkId=-1}
    ]
    blendTime [0.1,0.1,0.1,0.1,0.1,0.1]
   }
   Output Animation Pose(Result=@n5)
  ]
 }
 Transition {
  path=AnimGraph.AnimGraphNode_StateMachine_1.OnKillStateMachine.AnimStateTransitionNode_0.Transition
  nodes [
   n1 = Get Kill Counter() -> (Kill Counter=0)
   Result(bCanEnterTransition=AND Boolean(A=Get ActiveKill() -> (ActiveKill:bool), B="Equal (Integer)"(A=@n1, B=3)))
  ]
 }
 Transition {path=AnimGraph.AnimGraphNode_StateMachine_1.OnKillStateMachine.AnimStateTransitionNode_1.Transition, nodes [Result()]}
 Transition {
  path=AnimGraph.AnimGraphNode_StateMachine_1.OnKillStateMachine.AnimStateTransitionNode_3.Transition
  nodes [
   n1 = Get Kill Counter() -> (Kill Counter=0)
   Result(bCanEnterTransition=AND Boolean(A=Get ActiveKill(), B="Not Equal (Integer)"(A=@n1, B=3)))
   "Not Equal (Boolean)"(A:Wildcard, B:Wildcard) -> (ReturnValue=~)
  ]
 }
 EventGraph {nodes [
   n1 = Timer Exec() {is=title K2Node_CustomEvent}
   > Branch(Condition=Get Cosmetic Variant Reactivity())
   > Set activeKill(activeKill:bool=true) -> (Output_Get:bool)
   n2 = Get Kill Counter() -> (Kill Counter=0)
   n3 = Get MaxKills() -> (MaxKills:int=0)
   > Set Kill Counter(Kill Counter:int="Clamp (Integer)"(Value=int + int(A=@n2), Min=0, Max=int + int(A=@n3)) -> (ReturnValue=0)) -> (Output_Get:int=0)
   AnimNotify_Hide EggUpper() {is=title K2Node_Event}
   > SetScalarParameterValue(
    self=Get DynamicMaterialInstance() -> (DynamicMaterialInstance:Material Instance Dynamic Object Reference)
    ParameterName=Hide Element 04
    Value=1
   )
   > SetScalarParameterValue(self=Get DMISlotOverlay() -> (DMISlotOverlay:Material Instance Dynamic Object Reference), ParameterName=Hide Element 04, Value=1)
   Event OnCharacterCustomizationLoaded()
   n4 = GetPartSkeletalMeshComponent()
   > n5 = Get Dynamic Material Instance(SkeletalMeshComponent=@n4)
   > Set DynamicMaterialInstance(DynamicMaterialInstance:Material Instance Dynamic Object Reference=@n5) -> (Output_Get:Material Instance Dynamic Object Reference)
   > n6 = GetOrCreateDynamicMaterialSlotOverlayMaterial!(execute, MeshComponent:Mesh Component Object Reference=@n4, Index:int=0, MaterialName:Name=None) -> (then, ReturnValue:Material Instance Dynamic Object Reference)
   > Set DMISlotOverlay(DMISlotOverlay:Material Instance Dynamic Object Reference=@n6) -> (Output_Get:Material Instance Dynamic Object Reference)
   > SetMaterialVisibility!(execute, self, PomPom Vis:double=1, Whole Egg Vis:double=0, Egg Upper Vis:double=1, Egg Lower Vis:double=1) -> (then)
   > n7 = Cast To FortPlayerPawn(Object=TryGetFortPawn())
   > n8 = Cast To B_Athena_PartModifier_MixedAlley_Backpack!(
    execute
    Object=GetModifierForPartType!(self:Fort Player Pawn Object Reference=@n7, PartType:EFortCustomPartType Enum=Backpack) -> (ReturnValue:Custom Character Part Modifier Object Reference)
   ) -> (then, CastFailed, AsB Athena Part Modifier Mixed Alley Backpack:B Athena Part Modifier Mixed Alley Backpack Object Reference) {is=title K2Node_DynamicCast}
   > Set Cosmetic Variant Reactivity(Cosmetic Variant Reactivity:bool=Get Reactivity Enabled!(self:B Athena Part Modifier Mixed Alley Backpack Object Reference=@n8) -> (Reactivity Enabled:bool)) -> (Output_Get:bool)
   AnimNotify_Hide PomPom() {is=title K2Node_Event}
   > SetScalarParameterValue(self=Get DynamicMaterialInstance(), ParameterName=Hide Element 01, Value=1)
   > SetScalarParameterValue(self=Get DMISlotOverlay(), ParameterName=Hide Element 01, Value=1)
   n9 = Event BlueprintUpdateAnimation()
   > Set DeltaTimeX(DeltaTimeX:double=@n9.DeltaTimeX) -> (Output_Get:double)
   > n10 = Sequence!(execute) -> (then_0, then_1, then_2:Exec, then_3:Exec)
   n11 = Sequence!(execute=@n10.then_3) -> (then_0, then_1, then_2)
   > Branch(Condition=Get bIsInFrontEnd())
   > Branch(Condition="integer >= integer"(A=Get Kill Counter() -> (Kill Counter=0), B=4))
   > Do Once(Reset=@n15.then)
   > n12 = Delay(Duration=40)
   Do Once(execute=@n16.then, Reset=@n12.then)
   > Sequence!(execute) -> (then_0, then_1)
   > Set activeKill(activeKill=false)
   > Set Kill Counter(Kill Counter=0) -> (Output_Get=0)
   > n13 = Set Timer by Event(Delegate=@n1, Time=Get Timer() -> (Timer:float), bLooping=true)
   > n14 = Set TimerHandle(TimerHandle:Timer Handle Structure=@n13) -> (Output_Get:Timer Handle Structure)
   > n15 = Unpause Timer by Handle!(execute, Handle:Timer Handle Structure=@n14) -> (then)
   n16 = Branch(execute=@n11.then_1, Condition=Get bIsInFrontEnd())
   Branch(execute=@n11.then_2, Condition=Get bIsInFrontEnd())
   > Do Once(Reset=[@n19."  ", @n30.then])
   > n17 = Branch(Condition="integer > integer"(A=Get Kill Counter() -> (Kill Counter=0), B=3))
   > n18 = Delay(Duration=10)
   Branch(execute=[@n18.then, @n20.then, @n29.then], Condition=Get Cosmetic Variant Reactivity())
   > Set activeKill(activeKill=true)
   > n19 = Increment Int(Value=Get Kill Counter() -> (Kill Counter=0))
   n20 = Delay(execute=@n17.else, Duration=1.5)
   n21 = GetRealTimeSeconds()
   n22 = "Sin (Radians)"(A=float * float(A=2, B=@n21))
   n23 = "Sin (Radians)"(A=float * float(A=4, B=@n21))
   n24 = "Sin (Radians)"(A=float * float(A=2, B=@n21))
   Set FloatingBall_Movement(
    execute=@n10.then_0
    FloatingBall_Movement:Vector=MakeVector(X=float * float(A=@n23, B=0.9), Y=float * float(A=@n24, B=1), Z=float * float(A=@n22, B=1.2))
   ) -> (Output_Get:Vector)
   n25 = GetRealTimeSeconds()
   n26 = float * float(A="Sin (Radians)"(A=float * float(A=3, B=@n25)), B=1)
   Set StretchSquashFloatingBall(
    execute=@n10.then_1
    StretchSquashFloatingBall:Vector=MakeVector(
     X=MapRangeClamped(Value=@n26, InRangeA=1, InRangeB=-1, OutRangeA=0.95, OutRangeB=1.1)
     Y=MapRangeClamped(Value=@n26, InRangeA=1, InRangeB=-1, OutRangeA=0.95, OutRangeB=1.05)
     Z=MapRangeClamped(Value=@n26, InRangeA=-1, InRangeB=1, OutRangeA=0.95, OutRangeB=1.15)
    )
   ) -> (Output_Get:Vector)
   n27 = Branch(execute=@n10.then_2, Condition=Get bIsMoving2D())
   > Set Alpha_FloatingBall(Alpha_FloatingBall:double=0.2) -> (Output_Get:double)
   Set Alpha_FloatingBall(execute=@n27.else, Alpha_FloatingBall=1)
   AnimNotify_Hide EggLower() {is=title K2Node_Event}
   > SetScalarParameterValue(self=Get DynamicMaterialInstance(), ParameterName=Hide Element 03, Value=1)
   > SetScalarParameterValue(self=Get DMISlotOverlay(), ParameterName=Hide Element 03, Value=1)
   AnimNotify_Hide EggWhole() {is=title K2Node_Event}
   > SetScalarParameterValue(self=Get DynamicMaterialInstance(), ParameterName=Hide Element 02, Value=1)
   > SetScalarParameterValue(self=Get DMISlotOverlay(), ParameterName=Hide Element 02, Value=1)
   AnimNotify_Show EggUpper() {is=title K2Node_Event}
   > SetScalarParameterValue(self=Get DynamicMaterialInstance(), ParameterName=Hide Element 04, Value=0)
   > SetScalarParameterValue(self=Get DMISlotOverlay(), ParameterName=Hide Element 04, Value=0)
   AnimNotify_Show EggLower() {is=title K2Node_Event}
   > SetScalarParameterValue(self=Get DynamicMaterialInstance(), ParameterName=Hide Element 03, Value=0)
   > SetScalarParameterValue(self=Get DMISlotOverlay(), ParameterName=Hide Element 03, Value=0)
   n28 = Event BlueprintProcessAnimInputEvent()
   > n29 = Branch(Condition=IsNamed(self=@n28.AnimInputEvent, InEventName=aie.EnemyPlayerEliminated))
   AnimNotify_Show EggWhole() {is=title K2Node_Event}
   > SetScalarParameterValue(self=Get DynamicMaterialInstance(), ParameterName=Hide Element 02, Value=0)
   > SetScalarParameterValue(self=Get DMISlotOverlay(), ParameterName=Hide Element 02, Value=0)
   Event BlueprintPostEvaluateAnimation() {is=title K2Node_Event}
   > Branch(Condition=Get Cosmetic Variant Reactivity())
   > Set activeKill(activeKill=false)
   AnimNotify_ShowPomPom() {is=title K2Node_Event}
   > SetScalarParameterValue(self=Get DynamicMaterialInstance(), ParameterName=Hide Element 01, Value=0)
   > SetScalarParameterValue(self=Get DMISlotOverlay(), ParameterName=Hide Element 01, Value=0)
   Branch(Condition="integer > integer"(A=Get Kill Counter() -> (Kill Counter=0), B=Get MaxKills() -> (MaxKills=0)))
   > n30 = Set Kill Counter(Kill Counter=0) -> (Output_Get=0)
  ]}
 SetMaterialVisibility {nodes [
   SetMaterialVisibility!() -> (then, PomPom Vis:double=~, Whole Egg Vis:double=~, Egg Upper Vis:double=~, Egg Lower Vis:double=~) {is=title K2Node_FunctionEntry}
   n1 = Get DynamicMaterialInstance()
   n2 = Get PomPom Vis() -> (PomPom Vis:double)
   > SetScalarParameterValue(self=@n1, ParameterName=Hide Element 01, Value=@n2)
   n3 = Get Whole Egg Vis() -> (Whole Egg Vis:double)
   > SetScalarParameterValue(self=@n1, ParameterName=Hide Element 02, Value=@n3)
   n4 = Get Egg Upper Vis() -> (Egg Upper Vis:double)
   > SetScalarParameterValue(self=@n1, ParameterName=Hide Element 03, Value=@n4)
   n5 = Get Egg Lower Vis() -> (Egg Lower Vis:double)
   > SetScalarParameterValue(self=@n1, ParameterName=Hide Element 04, Value=@n5)
   n6 = Get DMISlotOverlay()
   > SetScalarParameterValue(self=@n6, ParameterName=Hide Element 01, Value=@n2)
   > SetScalarParameterValue(self=@n6, ParameterName=Hide Element 02, Value=@n3)
   > SetScalarParameterValue(self=@n6, ParameterName=Hide Element 03, Value=@n4)
   > SetScalarParameterValue(self=@n6, ParameterName=Hide Element 04, Value=@n5)
  ]}
]
variables [
 ActiveKill {type=boolean}
 Alpha_FloatingBall {type=number}
 Backward {type=number}
 BodyPosition {type=Vector}
 Cosmetic Variant Reactivity {type=boolean, default=true}
 CurrentPosition {type=Vector}
 DeltaTimeX {type=number}
 DMISlotOverlay {type=MaterialInstanceDynamic}
 dragDirection {type=Vector}
 DynamicMaterialInstance {type=MaterialInstanceDynamic}
 FloatingBall_Movement {type=Vector, default=<0.424 0 -0.4984>}
 Forward {type=number}
 Kill Counter {type=integer}
 Left {type=number}
 MaxKills {type=integer, default=5}
 MaxSpeed {type=number, default=500}
 moveBackward {type=number}
 moveDirection {type=Vector}
 moveForward {type=number}
 NormalizeSpeed {type=number}
 Right {type=number}
 Skeletal {type=SkeletalMeshComponent}
 SkelMesh {type=SkeletalMeshComponent}
 speed {type=number}
 SquashAmount {type=number}
 strafeLeft float {type=number}
 strafeRight float {type=number}
 StretchAmount {type=number}
 StretchIntensity {type=number, default=0.35}
 StretchSquashFloatingBall {type=Vector, default=<0.424 0 -0.4984>}
 Timer {type=number}
 TimerHandle {type=TimerHandle}
 velocity {type=Vector}
]
